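# Unlit translucent particle and Niagara sprite material for pistol weapon effects, producing a soft, scene-color distortion tinted by particle color.
# Dynamic-parameter Time scrolls noise that offsets scene color within a squared sphere mask, with white as the mobile fallback.
# Emissive passes through MF_FX_EmissiveAdjustments, while mask, particle alpha and depth fade form opacity.
# Separate branches compute a noisy screen-normal Fresnel depth blend and a camera offset, with no visible consumers.
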Material /Game/Weapons/FORT_Pistols/Effects/Weapon/M_DepthEmissive
graph {
 properties {
  bUsedWithNiagaraSprites=true
  bUsedWithParticleSprites=true
  blendMode=BLEND_Translucent
  pixelDepthOffsetMode=PDOM_Legacy
  shadingModel=MSM_Unlit
  translucencyPass=MTP_BeforeDOF
  usedShadingModels=Unlit
 }
 nodes [
  n1 = DotProduct(A=ComponentMask(None=SceneTexture() {sceneTextureId=PPI_WorldNormal}.Color) {b=true, g=true, r=true}, B=CameraVectorWS())
  n2 = OneMinus(None=Multiply(A=@n1, B=@n1))
  n3 = TextureCoordinate() {uTiling=3, vTiling=2}
  n4 = LinearInterpolate(Alpha=TextureSample(UVs=Panner(Coordinate=@n3) {speedX=-0.3, speedY=0.6}) {samplerType=SAMPLERTYPE_LinearColor, texture=&T_NoisePack_LinearColor}.G) {constA=4}
  LinearInterpolate(
   A=Multiply(A=Power(Base=@n2, Exp=@n4), B=VectorParameter() {parameterName=Color, defaultValue=<rgba 0.0208 0.5658 2 1>}.RGB)
   B=Multiply(A=SceneColor!(UVs).RGB, B=VectorParameter() {parameterName=Color2, defaultValue=<rgba 0.0289 0.136 0.151 1>}.RGB)
   Alpha=DepthFade() {fadeDistanceDefault=20}
  ) {constB=0}
  MaterialFunctionCall(Offset Amount=Constant() {r=5}) {materialFunction=/Engine/Functions/Engine_MaterialFunctions02/WorldPositionOffset/CameraOffset}
  n5 = Constant() {r=1}
  n6 = DynamicParameter() {paramNames [Time,Random,Param3,Param4]}
  n7 = Add(A=TextureCoordinate(), B=Multiply(A=Add(A=@n6.Time, B=@n6.Time), B=Constant2Vector() {g=1, r=0.01}))
  n8 = Subtract(A=ComponentMask(None=TextureSample(UVs=@n7) {samplerType=SAMPLERTYPE_LinearColor, texture=&T_NoisePack_LinearColor}.RGB) {g=true, r=true}) {constB=0.5}
  n9 = Constant() {r=0.5}
  n10 = SphereMask(A=TextureCoordinate(), B=@n9, Radius=@n9, Hardness=Constant())
  n11 = Multiply(A=@n10, B=@n10)
  n12 = Multiply(B=@n11) {constA=0.012}
  n13 = ShadingPathSwitch(Default=SceneColor!(OffsetFraction=Multiply(A=@n8, B=@n12) {constB=0.1}) {inputMode=OffsetFraction}.RGB, Mobile=@n5) {desc=Migrated from FeatureLevelSwitch}
  n14 = ScalarParameter() {defaultValue=1, parameterName=UseSceneColor, sliderMax=1}
  n15 = ParticleColor()
  MaterialFunctionCall(Emissive=Multiply(A=LinearInterpolate(A=@n5, B=@n13, Alpha=@n14), B=@n15.RGB)) {materialFunction=&MF_FX_EmissiveAdjustments}
  Multiply(A=Multiply(A=@n11, B=@n15.A), B=DepthFade(FadeDistance=ScalarParameter() {defaultValue=5, parameterName=DF_FD}) {fadeDistanceDefault=5})
 ]
}
