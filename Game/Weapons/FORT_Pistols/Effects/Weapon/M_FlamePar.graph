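# Translucent unlit sprite material for flame particles in pistol weapon effects, usable by Cascade and Niagara.
# Noise-distorted panning streaks, timed by dynamic parameter or game time, combine with a sphere mask, then particle alpha is subtracted to erode opacity.
# Particle color tints noise-refracted scene color, modulated by a second streak sample for emissive.
# Mobile skips depth fade and refraction; optional switches add world position offset and dynamic alpha.
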Material /Game/Weapons/FORT_Pistols/Effects/Weapon/M_FlamePar
graph {
 properties {
  bUsedWithNiagaraSprites=true
  bUsedWithParticleSprites=true
  blendMode=BLEND_Translucent
  pixelDepthOffsetMode=PDOM_Legacy
  shadingModel=MSM_Unlit
  translucencyPass=MTP_BeforeDOF
  usedShadingModels=Unlit
 }
 nodes [
  n1 = TextureCoordinate() {uTiling=2, vTiling=0.5}
  n2 = ShadingPathSwitch(Default=ParticleRandom(), Mobile=Constant() {r=1}) {desc=Migrated from FeatureLevelSwitch}
  n3 = Time()
  n4 = DynamicParameter() {paramNames [Time,Random,Alpha,Param4]}
  n5 = StaticSwitchParameter(True=Multiply(A=Add(A=@n2, B=@n3), B=ScalarParameter() {parameterName=TimeSpeed}), False=Add(A=@n4.Time, B=@n4.Time)) {parameterName=AsGPU}
  n6 = TextureSample() {samplerType=SAMPLERTYPE_LinearColor, texture=&T_NoisePack_LinearColor}
  n7 = Multiply(A=Subtract(A=ComponentMask(None=@n6.RGB) {g=true, r=true}) {constB=0.5}) {constB=0.1}
  n8 = Multiply(
   A=Add(A=Panner(Coordinate=@n1, Time=@n5) {speedX=0.1, speedY=0.5}, B=@n7)
   B=ScalarParameter() {defaultValue=1, parameterName=TextureUVScale}
  )
  n9 = TextureSample(UVs=@n8) {texture=/Game/Effects/Fort_Effects/Textures/Energy/T_OutpostShockwaveStreaks_02}
  n10 = MaterialFunctionCall(Float2=TextureCoordinate()) {materialFunction=&BreakOutFloat2Components}
  n11 = SphereMask(A=AppendVector(A=@n10.R, B=Multiply(A=@n10.G, B=@n10.G)), B=Constant() {r=0.5}, Radius=Constant() {r=0.5}, Hardness=Constant() {r=-1})
  n12 = Subtract(A=Add(A=Power(Base=@n9.G) {constExponent=0.2}, B=@n11), B=OneMinus(None=@n11))
  n13 = ParticleColor()
  n14 = Subtract(A=@n12, B=@n13.A)
  n15 = DepthFade(FadeDistance=ScalarParameter() {defaultValue=10, parameterName=DepthFadeDistance}) {fadeDistanceDefault=10}
  n16 = Multiply(A=@n15, B=@n15)
  n17 = Constant() {r=1}
  n18 = Multiply(
   A=Multiply(A=Multiply(A=@n14) {constB=2}, B=ShadingPathSwitch(Default=@n16, Mobile=@n17) {desc=Migrated from FeatureLevelSwitch})
   B=ScalarParameter() {defaultValue=0.75, parameterName=Opacity}
  ) {constB=0.75}
  LinearInterpolate(
   A=@n18
   B=Multiply(A=@n18, B=DynamicParameter() {paramNames [Time,Random,Alpha,Param4]}.Alpha)
   Alpha=ScalarParameter() {parameterName=UseDynamicParamAlpha}
  )
  n19 = Multiply(A=Multiply(A=@n7) {constB=0.5}, B=Saturate(None=@n12))
  n20 = Multiply(
   A=ShadingPathSwitch(Default=SceneColor!(OffsetFraction=@n19) {inputMode=OffsetFraction}.RGB, Mobile=Constant() {r=1}) {desc=Migrated from FeatureLevelSwitch}
   B=@n13.RGB
  )
  n21 = TextureSample(UVs=Multiply(A=@n8) {constB=2.427}) {texture=&T_OutpostShockwaveStreaks_02}
  MaterialFunctionCall(Emissive=Multiply(A=LinearInterpolate(Alpha=@n21.G) {constA=0.5, constB=1.5}, B=LinearInterpolate(A=@n13.RGB, B=@n20) {constAlpha=0.9})) {materialFunction=&MF_FX_EmissiveAdjustments}
  n22 = VectorParameter() {parameterName=WPO_Direction}
  n23 = ScalarParameter() {defaultValue=1, parameterName=WPO_Intencity}
  StaticSwitchParameter(True=Multiply(A=Multiply(A=@n22.RGB, B=@n23), B=DynamicParameter() {paramNames [Time,Random,Alpha,Param4]}.Alpha), False=Constant()) {parameterName=UseWPO}
 ]
}
