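# Translucent, unlit, two-sided mesh-particle material, named for a sci-fi pistol ring effect.
# An offset along vertex normals combines RingSize with MeshWarble, optionally modulated by panning electricity noise, all driven per particle through dynamic parameters.
# Particle color, optionally hue-shifted by panning noise, passes through a shared emissive adjustment function.
# Particle alpha can be multiplied by depth fade, bypassed on mobile, and by panning noise through static switches.
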Material /Game/Weapons/FORT_Pistols/Effects/Weapon/M_Pistol_Scifi_Ring
graph {
 properties {
  bUsedWithMeshParticles=true
  blendMode=BLEND_Translucent
  pixelDepthOffsetMode=PDOM_Legacy
  shadingModel=MSM_Unlit
  twoSided=true
  usedShadingModels=Unlit
 }
 nodes [
  n1 = VertexNormalWS()
  n2 = DynamicParameter() {defaultValue=<rgba 0.5 1 1 1>, paramNames [XSpeed,RingSize,Rand,MeshWarble]}
  n3 = DynamicParameter() {defaultValue=<rgba 0.5 1 1 1>, paramNames [XSpeed,RingSize,Rand,MeshWarble]}
  n4 = Constant() {r=0.1}
  n5 = MaterialFunctionCall(Float2=Panner(Coordinate=TextureCoordinate(), Speed=AppendVector(A=@n3.XSpeed, B=@n4)) {speedX=1, speedY=0.3}) {materialFunction=&BreakOutFloat2Components}
  n6 = Multiply(A=ScalarParameter() {defaultValue=1, parameterName=U_Tile_Noise}, B=Add(A=@n3.Rand, B=@n5.R))
  n7 = Multiply(A=@n5.G, B=ScalarParameter() {defaultValue=1, parameterName=V_Tile_Noise})
  n8 = LinearInterpolate(Alpha=TextureSample(UVs=AppendVector(A=@n6, B=@n7)) {samplerType=SAMPLERTYPE_Masks, texture=/Game/Characters/Player/Male/Medium/Bodies/M_MED_Storm_Soldier/Textures/ElectricityNoise_M}.G) {constA=-0.5}
  n9 = StaticSwitchParameter(True=ConstantBiasScale(None=Saturate(None=@n8)) {bias=-0.5, scale=2}, False=Constant() {r=1}) {defaultValue=true, parameterName=UseAnimated Warble}
  Add(A=Multiply(A=@n1, B=@n2.RingSize), B=Multiply(A=@n1, B=Multiply(A=@n2.MeshWarble, B=@n9)))
  n10 = TextureCoordinate()
  n11 = Multiply(A=ScalarParameter() {defaultValue=1, group=ColorNoise, parameterName=U_Tile_ColorNoise}, B=ComponentMask(None=@n10) {r=true})
  n12 = Multiply(A=ComponentMask(None=@n10) {g=true}, B=ScalarParameter() {defaultValue=1, group=ColorNoise, parameterName=V_Tile_ColorNoise})
  n13 = ScalarParameter() {defaultValue=0.1, group=ColorNoise, parameterName=Hueshift Panning Speed - X}
  n14 = ScalarParameter() {defaultValue=0.1, group=ColorNoise, parameterName=Hueshift Panning Speed - Y}
  n15 = TextureSampleParameter2D(UVs=Panner(Coordinate=AppendVector(A=@n11, B=@n12), Speed=AppendVector(A=@n13, B=@n14)) {speedY=1}) {
   group=ColorNoise
   parameterName=HueshiftNoise
   samplerType=SAMPLERTYPE_LinearColor
   texture=/Game/Global/Textures/Default/Noise/T_Noise01_Inverse
  }
  n16 = Clamp(None=Power(Base=@n15.R, Exp=ScalarParameter() {defaultValue=3, group=ColorNoise, parameterName=Hueshift Noise Exponent}))
  n17 = ParticleColor()
  MaterialFunctionCall(Emissive=StaticSwitchParameter(
    True=MaterialFunctionCall(Hue Shift Percentage=@n16, Texture=@n17.RGB) {materialFunction=&HueShift}.Result
    False=@n17.RGB
   ) {group=ColorNoise, parameterName=Use Hueshift Noise}) {materialFunction=&MF_FX_EmissiveAdjustments}
  n18 = Multiply(
   A=DepthFade(Opacity=Constant() {r=1}, FadeDistance=ScalarParameter() {defaultValue=1, parameterName=FadeDistance})
   B=ScalarParameter() {defaultValue=1, parameterName=DepthFade-Hardness}
  )
  n19 = Constant() {r=1}
  n20 = StaticSwitchParameter(True=Multiply(A=@n17.A, B=ShadingPathSwitch(Default=@n18, Mobile=@n19) {desc=Migrated from FeatureLevelSwitch}), False=@n17.A) {parameterName=UseDepthFade}
  n21 = TextureCoordinate() {uTiling=0.05}
  n22 = AppendVector(A=Add(A=ComponentMask(None=@n21) {r=true}), B=Add(A=ComponentMask(None=@n21) {g=true}))
  n23 = AppendVector(A=ScalarParameter() {defaultValue=0.1, parameterName=Alpha Panning Speed}, B=Constant())
  n24 = Power(
   Base=TextureSampleParameter2D(UVs=Panner(Coordinate=@n22, Speed=@n23) {speedY=1}) {parameterName=Alpha Noise, samplerType=SAMPLERTYPE_LinearColor, texture=&T_Noise01_Inverse}.R
   Exp=ScalarParameter() {defaultValue=6, parameterName=Alpha Noise Exponent}
  )
  StaticSwitchParameter(True=Multiply(A=@n20, B=Clamp(None=@n24)), False=@n20) {parameterName=UseAlphaNoise}
 ]
}
